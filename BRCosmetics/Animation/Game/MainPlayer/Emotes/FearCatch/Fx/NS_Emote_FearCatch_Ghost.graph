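# Niagara ribbon effect for the FearCatch emote, drawing three ghosts that get pulled toward a thermos.
# Ghost1–3 each trail one ghost, while Ghost4 adds a green, camera-offset ribbon for all three by spawn group.
# Ribbons start at ghost skeletal mesh sockets, and per-ghost User GhostLerp values bend them along Hermite splines to the thermos and reshape their width.
# GhostSpawn values scale spawn rates, and GhostLerp toggles angry/scared color alphas around 0.5.
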NiagaraSystem /BRCosmetics/Animation/Game/MainPlayer/Emotes/FearCatch/Fx/NS_Emote_FearCatch_Ghost
dependencies {
 usedDataInterfaces [&NiagaraDataInterfaceCurve, &NiagaraDataInterfaceColorCurve, &NiagaraDataInterfaceSkeletalMesh]
 usedDynamicInputs []
 usedModules [
  &self:SystemSpawnScript.SystemScriptSource.SystemScriptGraph.NiagaraNodeAssignment_0.SetVariables_#1
  &SystemState
  &self:SystemSpawnScript.SystemScriptSource.SystemScriptGraph.NiagaraNodeAssignment_1.SetVariables_#2
 ]
}
emitters {
 Ghost1 {
  emitterUpdateScript [
   &EmitterState(Life Cycle Mode=::NewEnumerator1, Loop Behavior=::NewEnumerator1, Loop Duration Mode=::NewEnumerator1, UseLoopDelay=false)
   &SpawnRate(SpawnRate=&Multiply_Float(A=30, B=$User.GhostSpawn1), Use Spawn Probability=false)
  ]
  particleSpawnScript [
   /SetVariables(Particles.RibbonID=/Niagara/DynamicInputs/Meshes/MakeRibbonIDFromSpawnGroup()) {isSetParameters=true, module=SetVariables_#3}
   &InitializeParticle(
    Color=<rgba 1 1 1 0.1913>
    Lifetime=0.4
    Ribbon Width=&MergeHelper_FloatSwitch(0=35, 1=35, 2=35, 3=?, Index=&SpawnGroup(), Num=::NewEnumerator1)
    Ribbon Width Mode=::NewEnumerator1
    UsePositionOffset=false
    Write Material Random=true
   )
   &SkeletalMeshLocation(
    Bone / Socket Index=&MergeHelper_IntSwitch(0=0, 1=1, 2=2, 3=?, 4=?, 5=?, Index=$Particles.SpawnGroup, Num=::NewEnumerator1)
    MaskSpawnGroup=false
    Mesh Sampling Type=::NewEnumerator5
    Position Sampling=::NewEnumerator2
    Skeletal Mesh=$System.GhostSkelMesh
    Socket Sampling Mode=::NewEnumerator5
    Use Sampled Mesh Attribute as Ribbon ID=false
   )
  ]
  particleUpdateScript [
   &SkeletalMeshLocation(
    MaskSpawnGroup=false
    Mesh Sampling Type=::NewEnumerator5
    Position Sampling=::NewEnumerator2
    Sampled Position Offset=&MakeVector(
     X=0
     Y=&Add_Float(
      A=&Lerp_Float(A=60, B=160, Alpha=$User.GhostLerp1)
      B=&Waveform("Waveform [1]"=::NewEnumerator9, "[1] Frequency"=1, "[1] Amplitude Min/Max"=<0 1>, "[1] Scale"=35.1096)
     )
     Z=0
    )
    Skeletal Mesh=$System.ThermosSkelMesh
    Socket Sampling Mode=::NewEnumerator5
    Use Sampled Mesh Attribute as Ribbon ID=false
   ) {module=SkeletalMeshLocation001}
   /Game/Effects/Niagara/Modules/Curves/FN_HermiteSpline(
    Direction End:Vector3f=&Multiply_VectorByFloat(
     Vector=Linked() {linkedVariable {name=Output.SkeletalMeshLocation001.SampledPositionWithoutOffset, type=NiagaraPosition}}
     Float=&Multiply_Float(
      A=-1
      B=&MergeHelper_FloatSwitch(0=$User.GhostLerp1, 1=$User.GhostLerp2, 2=$User.GhostLerp3, 3=?, Index=&SpawnGroup(), Num=::NewEnumerator1)
     )
    )
    Position Blend:NiagaraFloat=$Particles.NormalizedAge
    Position End:Vector3f=&Lerp_Vector(
     A=Linked() {linkedVariable {name=Particles.Previous.Position, type=NiagaraPosition}}
     B=Linked() {linkedVariable {name=Output.SkeletalMeshLocation001.SampledPositionWithoutOffset, type=NiagaraPosition}}
     Alpha=&FloatFromCurve(
      FloatCurve=DataInterface() {propertyValues {Curve {
         defaultValue=1
         keys [
          {interpMode=RCIM_Cubic, leaveTangent=1.6631, tangentMode=RCTM_Break, time=0, value=0.076}
          {interpMode=RCIM_Cubic, tangentMode=RCTM_Break, time=1, value=1}
         ]
        }}}
      Scale Curve=&MergeHelper_FloatSwitch(0=$User.GhostLerp1, 1=$User.GhostLerp2, 2=$User.GhostLerp3, 3=?, Index=&SpawnGroup(), Num=::NewEnumerator1)
     )
    )
    Position Start:Vector3f=&Lerp_Vector(
     A=Linked() {linkedVariable {name=Particles.Initial.SkeletalMeshLocation.SampledPosition, type=NiagaraPosition}}
     B=Linked() {linkedVariable {name=Particles.SkeletalMeshLocation001.SampledPosition, type=NiagaraPosition}}
     Alpha=&MergeHelper_FloatSwitch(0=$User.GhostLerp1, 1=$User.GhostLerp2, 2=$User.GhostLerp3, 3=?, Index=&SpawnGroup(), Num=::NewEnumerator1)
    )
    Write Particles Position=true
   )
   &ParticleState()
   &SolveForcesAndVelocity(Clamp Velocity=false, Limit Acceleration=false)
   &ScaleRibbonWidth(Ribbon Width Scale=&Lerp_Float(
     A=&FloatFromCurve(FloatCurve=DataInterface() {propertyValues {Curve {keys [<at 0 1.7>,<at 1 0.4278>]}}}, Scale Curve=2)
     B=&FloatFromCurve(
      FloatCurve=DataInterface() {propertyValues {Curve {keys [
          {interpMode=RCIM_Cubic, leaveTangent=-3.8671, tangentMode=RCTM_Break, time=0.4757, value=2.7348}
          {arriveTangent=-1.3032, interpMode=RCIM_Cubic, leaveTangent=-1.3032, tangentMode=RCTM_User, time=0.709, value=0.7882}
          <at 1 0.4278>
         ]}}}
      Scale Curve=2
     )
     Alpha=&MergeHelper_FloatSwitch(0=$User.GhostLerp1, 1=$User.GhostLerp2, 2=$User.GhostLerp3, 3=?, Index=&SpawnGroup(), Num=::NewEnumerator1)
    ))
   &DynamicMaterialParameters(Module.Index0Float0Write=true, Module.Index0Float1Write=true, Module.Index0Float2Write=true, Module.Index0Float3Write=true)
   /SetVariables(
    Particles.AngryColor:LinearColor=&MakeCustomLinearColorFromBool(
     Bool=&SetBoolByFloatComparison(A=$User.GhostLerp1, B=0.5)
     True Color=<rgba 1 1 1 0>
     False Color=<rgba 1 1 1 0.8>
    )
    Particles.ScaredColor:LinearColor=&MakeCustomLinearColorFromBool(
     Bool=&SetBoolByFloatComparison(A=$User.GhostLerp1, B=0.5)
     True Color=<rgba 1 1 1 0.8>
     False Color=<rgba 1 1 1 0>
    )
   ) {isSetParameters=true, module=SetVariables_#4}
  ]
  rendererClasses [&NiagaraRibbonRendererProperties]
  renderers [
   {rendererClass=&NiagaraRibbonRendererProperties}
   {rendererClass=&NiagaraRibbonRendererProperties, rendererIndex=1}
  ]
 }
 Ghost2 {
  emitterUpdateScript [
   &EmitterState(Life Cycle Mode=::NewEnumerator1, Loop Behavior=::NewEnumerator1, Loop Duration Mode=::NewEnumerator1, UseLoopDelay=false)
   &SpawnRate(SpawnGroup=1, SpawnRate=&Multiply_Float(A=30, B=$User.GhostSpawn2), Use Spawn Probability=false) {module=SpawnRate001}
  ]
  particleSpawnScript [
   /SetVariables(Particles.RibbonID=&MakeRibbonIDFromSpawnGroup()) {isSetParameters=true, module=SetVariables_#3}
   &InitializeParticle(
    Color=<rgba 1 1 1 0.8>
    Lifetime=0.4
    Ribbon Width=&MergeHelper_FloatSwitch(0=32, 1=30, 2=35, 3=?, Index=&SpawnGroup(), Num=::NewEnumerator1)
    Ribbon Width Mode=::NewEnumerator1
    UsePositionOffset=false
    Write Material Random=true
   )
   &SkeletalMeshLocation(
    Bone / Socket Index=&MergeHelper_IntSwitch(0=0, 1=1, 2=2, 3=?, 4=?, 5=?, Index=$Particles.SpawnGroup, Num=::NewEnumerator1)
    MaskSpawnGroup=false
    Mesh Sampling Type=::NewEnumerator5
    Position Sampling=::NewEnumerator2
    Skeletal Mesh=$System.GhostSkelMesh
    Socket Sampling Mode=::NewEnumerator5
    Use Sampled Mesh Attribute as Ribbon ID=false
   )
  ]
  particleUpdateScript [
   &SkeletalMeshLocation(
    MaskSpawnGroup=false
    Mesh Sampling Type=::NewEnumerator5
    Position Sampling=::NewEnumerator2
    Sampled Position Offset=&MakeVector(
     X=0
     Y=&Add_Float(
      A=&Lerp_Float(A=60, B=160, Alpha=$User.GhostLerp2)
      B=&Waveform("Waveform [1]"=::NewEnumerator9, "[1] Frequency"=1, "[1] Amplitude Min/Max"=<0 1>, "[1] Scale"=35.1096)
     )
     Z=0
    )
    Skeletal Mesh=$System.ThermosSkelMesh
    Socket Sampling Mode=::NewEnumerator5
    Use Sampled Mesh Attribute as Ribbon ID=false
   ) {module=SkeletalMeshLocation001}
   &FN_HermiteSpline(
    Direction End:Vector3f=&Multiply_VectorByFloat(
     Vector=Linked() {linkedVariable {name=Output.SkeletalMeshLocation001.SampledPositionWithoutOffset, type=NiagaraPosition}}
     Float=&Multiply_Float(
      A=-1
      B=&MergeHelper_FloatSwitch(0=$User.GhostLerp1, 1=$User.GhostLerp2, 2=$User.GhostLerp3, 3=?, Index=&SpawnGroup(), Num=::NewEnumerator1)
     )
    )
    Position Blend:NiagaraFloat=$Particles.NormalizedAge
    Position End:Vector3f=&Lerp_Vector(
     A=Linked() {linkedVariable {name=Particles.Previous.Position, type=NiagaraPosition}}
     B=Linked() {linkedVariable {name=Output.SkeletalMeshLocation001.SampledPositionWithoutOffset, type=NiagaraPosition}}
     Alpha=&FloatFromCurve(
      FloatCurve=DataInterface() {propertyValues {Curve {
         defaultValue=1
         keys [
          {interpMode=RCIM_Cubic, leaveTangent=1.6631, tangentMode=RCTM_Break, time=0, value=0.076}
          {interpMode=RCIM_Cubic, tangentMode=RCTM_Break, time=1, value=1}
         ]
        }}}
      Scale Curve=&MergeHelper_FloatSwitch(0=$User.GhostLerp1, 1=$User.GhostLerp2, 2=$User.GhostLerp3, 3=?, Index=&SpawnGroup(), Num=::NewEnumerator1)
     )
    )
    Position Start:Vector3f=&Lerp_Vector(
     A=Linked() {linkedVariable {name=Particles.Initial.SkeletalMeshLocation.SampledPosition, type=NiagaraPosition}}
     B=Linked() {linkedVariable {name=Particles.SkeletalMeshLocation001.SampledPosition, type=NiagaraPosition}}
     Alpha=&MergeHelper_FloatSwitch(0=$User.GhostLerp1, 1=$User.GhostLerp2, 2=$User.GhostLerp3, 3=?, Index=&SpawnGroup(), Num=::NewEnumerator1)
    )
    Write Particles Position=true
   )
   &ParticleState()
   &SolveForcesAndVelocity(Clamp Velocity=false, Limit Acceleration=false)
   &ScaleRibbonWidth(Ribbon Width Scale=&Lerp_Float(
     A=&FloatFromCurve(FloatCurve=DataInterface() {propertyValues {Curve {keys [<at 0 1.7>,<at 1 0.4278>]}}}, Scale Curve=2)
     B=&FloatFromCurve(
      FloatCurve=DataInterface() {propertyValues {Curve {keys [
          {interpMode=RCIM_Cubic, leaveTangent=-3.8671, tangentMode=RCTM_Break, time=0.4757, value=2.7348}
          {arriveTangent=-1.3032, interpMode=RCIM_Cubic, leaveTangent=-1.3032, tangentMode=RCTM_User, time=0.709, value=0.7882}
          <at 1 0.4278>
         ]}}}
      Scale Curve=2
     )
     Alpha=&MergeHelper_FloatSwitch(0=$User.GhostLerp1, 1=$User.GhostLerp2, 2=$User.GhostLerp3, 3=?, Index=&SpawnGroup(), Num=::NewEnumerator1)
    ))
   &DynamicMaterialParameters(Module.Index0Float0Write=true, Module.Index0Float1Write=true, Module.Index0Float2Write=true, Module.Index0Float3Write=true)
   /SetVariables(
    Particles.AngryColor:LinearColor=&MakeCustomLinearColorFromBool(
     Bool=&SetBoolByFloatComparison(A=$User.GhostLerp2, B=0.5)
     True Color=<rgba 1 1 1 0>
     False Color=<rgba 1 1 1 0.8>
    )
    Particles.ScaredColor:LinearColor=&MakeCustomLinearColorFromBool(
     Bool=&SetBoolByFloatComparison(A=$User.GhostLerp2, B=0.5)
     True Color=<rgba 1 1 1 0.8>
     False Color=<rgba 1 1 1 0>
    )
   ) {isSetParameters=true}
  ]
  rendererClasses [&NiagaraRibbonRendererProperties]
  renderers [
   {rendererClass=&NiagaraRibbonRendererProperties}
   {rendererClass=&NiagaraRibbonRendererProperties, rendererIndex=1}
  ]
 }
 Ghost3 {
  emitterUpdateScript [
   &EmitterState(Life Cycle Mode=::NewEnumerator1, Loop Behavior=::NewEnumerator1, Loop Duration Mode=::NewEnumerator1, UseLoopDelay=false)
   &SpawnRate(SpawnGroup=2, SpawnRate=&Multiply_Float(A=30, B=$User.GhostSpawn3), Use Spawn Probability=false) {module=SpawnRate002}
  ]
  particleSpawnScript [
   /SetVariables(Particles.RibbonID=&MakeRibbonIDFromSpawnGroup()) {isSetParameters=true, module=SetVariables_#3}
   &InitializeParticle(
    Color=<rgba 1 1 1 0.8>
    Lifetime=0.4
    Ribbon Width=&MergeHelper_FloatSwitch(0=35, 1=25, 2=30, 3=?, Index=&SpawnGroup(), Num=::NewEnumerator1)
    Ribbon Width Mode=::NewEnumerator1
    UsePositionOffset=false
    Write Material Random=true
   )
   &SkeletalMeshLocation(
    Bone / Socket Index=&MergeHelper_IntSwitch(0=0, 1=1, 2=2, 3=?, 4=?, 5=?, Index=$Particles.SpawnGroup, Num=::NewEnumerator1)
    MaskSpawnGroup=false
    Mesh Sampling Type=::NewEnumerator5
    Position Sampling=::NewEnumerator2
    Skeletal Mesh=$System.GhostSkelMesh
    Socket Sampling Mode=::NewEnumerator5
    Use Sampled Mesh Attribute as Ribbon ID=false
   )
  ]
  particleUpdateScript [
   &ParticleState()
   &SkeletalMeshLocation(
    MaskSpawnGroup=false
    Mesh Sampling Type=::NewEnumerator5
    Position Sampling=::NewEnumerator2
    Sampled Position Offset=&MakeVector(
     X=0
     Y=&Add_Float(
      A=&Lerp_Float(A=60, B=120, Alpha=$User.GhostLerp3)
      B=&Waveform("Waveform [1]"=::NewEnumerator9, "[1] Frequency"=1, "[1] Amplitude Min/Max"=<0 1>, "[1] Scale"=35.1096)
     )
     Z=0
    )
    Skeletal Mesh=$System.ThermosSkelMesh
    Socket Sampling Mode=::NewEnumerator5
    Use Sampled Mesh Attribute as Ribbon ID=false
   ) {module=SkeletalMeshLocation001}
   &FN_HermiteSpline(
    Direction End:Vector3f=&Multiply_VectorByFloat(
     Vector=Linked() {linkedVariable {name=Output.SkeletalMeshLocation001.SampledPositionWithoutOffset, type=NiagaraPosition}}
     Float=&Multiply_Float(
      A=-1
      B=&MergeHelper_FloatSwitch(0=$User.GhostLerp1, 1=$User.GhostLerp2, 2=$User.GhostLerp3, 3=?, Index=&SpawnGroup(), Num=::NewEnumerator1)
     )
    )
    Position Blend:NiagaraFloat=$Particles.NormalizedAge
    Position End:Vector3f=&Lerp_Vector(
     A=Linked() {linkedVariable {name=Particles.Previous.Position, type=NiagaraPosition}}
     B=Linked() {linkedVariable {name=Output.SkeletalMeshLocation001.SampledPositionWithoutOffset, type=NiagaraPosition}}
     Alpha=&FloatFromCurve(
      FloatCurve=DataInterface() {propertyValues {Curve {
         defaultValue=1
         keys [
          {interpMode=RCIM_Cubic, leaveTangent=5.7563, tangentMode=RCTM_Break, time=0, value=0.076}
          {arriveTangent=1.1864, interpMode=RCIM_Cubic, leaveTangent=1.1864, tangentMode=RCTM_User, time=0.1792, value=0.7527}
          {interpMode=RCIM_Cubic, leaveTangent=0.1168, tangentMode=RCTM_Break, time=1, value=1}
         ]
        }}}
      Scale Curve=&MergeHelper_FloatSwitch(0=$User.GhostLerp1, 1=$User.GhostLerp2, 2=$User.GhostLerp3, 3=?, Index=&SpawnGroup(), Num=::NewEnumerator1)
     )
    )
    Position Start:Vector3f=&Lerp_Vector(
     A=Linked() {linkedVariable {name=Particles.Initial.SkeletalMeshLocation.SampledPosition, type=NiagaraPosition}}
     B=Linked() {linkedVariable {name=Particles.SkeletalMeshLocation001.SampledPosition, type=NiagaraPosition}}
     Alpha=&MergeHelper_FloatSwitch(0=$User.GhostLerp1, 1=$User.GhostLerp2, 2=$User.GhostLerp3, 3=?, Index=&SpawnGroup(), Num=::NewEnumerator1)
    )
    Write Particles Position=true
   )
   &SolveForcesAndVelocity(Clamp Velocity=false, Limit Acceleration=false)
   &ScaleRibbonWidth(Ribbon Width Scale=&Lerp_Float(
     A=&FloatFromCurve(FloatCurve=DataInterface() {propertyValues {Curve {keys [<at 0 1.7>,<at 1 0.4278>]}}}, Scale Curve=2)
     B=&FloatFromCurve(
      FloatCurve=DataInterface() {propertyValues {Curve {keys [
          {interpMode=RCIM_Cubic, leaveTangent=-3.8671, tangentMode=RCTM_Break, time=0.4776, value=2.2136}
          {arriveTangent=-1.3032, interpMode=RCIM_Cubic, leaveTangent=-1.3032, tangentMode=RCTM_User, time=0.5769, value=0.8756}
          <at 1 0.4278>
         ]}}}
      Scale Curve=2
     )
     Alpha=&MergeHelper_FloatSwitch(0=$User.GhostLerp1, 1=$User.GhostLerp2, 2=$User.GhostLerp3, 3=?, Index=&SpawnGroup(), Num=::NewEnumerator1)
    ))
   &DynamicMaterialParameters(Module.Index0Float0Write=true, Module.Index0Float1Write=true, Module.Index0Float2Write=true, Module.Index0Float3Write=true)
   /SetVariables(
    Particles.AngryColor:LinearColor=&MakeCustomLinearColorFromBool(
     Bool=&SetBoolByFloatComparison(A=$User.GhostLerp3, B=0.5)
     True Color=<rgba 1 1 1 0>
     False Color=<rgba 1 1 1 0.8>
    )
    Particles.ScaredColor:LinearColor=&MakeCustomLinearColorFromBool(
     Bool=&SetBoolByFloatComparison(A=$User.GhostLerp3, B=0.5)
     True Color=<rgba 1 1 1 0.8>
     False Color=<rgba 1 1 1 0>
    )
   ) {isSetParameters=true}
  ]
  rendererClasses [&NiagaraRibbonRendererProperties]
  renderers [
   {rendererClass=&NiagaraRibbonRendererProperties}
   {rendererClass=&NiagaraRibbonRendererProperties, rendererIndex=1}
  ]
 }
 Ghost4 {
  emitterUpdateScript [
   &EmitterState(Life Cycle Mode=::NewEnumerator1, Loop Behavior=::NewEnumerator1, Loop Duration Mode=::NewEnumerator1, UseLoopDelay=false)
   &SpawnRate(SpawnRate=&Multiply_Float(A=50, B=$User.GhostSpawn1), Use Spawn Probability=false)
   &SpawnRate(SpawnGroup=1, SpawnRate=&Multiply_Float(A=50, B=$User.GhostSpawn2), Use Spawn Probability=false) {module=SpawnRate001}
   &SpawnRate(SpawnGroup=2, SpawnRate=&Multiply_Float(A=50, B=$User.GhostSpawn3), Use Spawn Probability=false) {module=SpawnRate002}
  ]
  particleSpawnScript [
   /SetVariables(Particles.RibbonID=&MakeRibbonIDFromSpawnGroup()) {isSetParameters=true, module=SetVariables_#3}
   &InitializeParticle(
    Color=<rgba 0.131 1 0 0.8609>
    Lifetime=0.6
    Ribbon Width=&MergeHelper_FloatSwitch(0=50, 1=50, 2=50, 3=?, Index=&SpawnGroup(), Num=::NewEnumerator1)
    Ribbon Width Mode=::NewEnumerator1
    UsePositionOffset=false
    Write Material Random=true
   )
   &SkeletalMeshLocation(
    Bone / Socket Index=&MergeHelper_IntSwitch(0=0, 1=1, 2=2, 3=?, 4=?, 5=?, Index=$Particles.SpawnGroup, Num=::NewEnumerator1)
    MaskSpawnGroup=false
    Mesh Sampling Type=::NewEnumerator5
    Position Sampling=::NewEnumerator2
    Skeletal Mesh=$System.GhostSkelMesh
    Socket Sampling Mode=::NewEnumerator5
    Use Sampled Mesh Attribute as Ribbon ID=false
   )
  ]
  particleUpdateScript [
   &SkeletalMeshLocation(
    MaskSpawnGroup=false
    Mesh Sampling Type=::NewEnumerator5
    Position Sampling=::NewEnumerator2
    Sampled Position Offset=&MakeVector(
     X=0
     Y=&Add_Float(
      A=&Lerp_Float(A=60, B=160, Alpha=$User.GhostLerp1)
      B=&Waveform("Waveform [1]"=::NewEnumerator9, "[1] Frequency"=1, "[1] Amplitude Min/Max"=<0 1>, "[1] Scale"=35.1096)
     )
     Z=0
    )
    Skeletal Mesh=$System.ThermosSkelMesh
    Socket Sampling Mode=::NewEnumerator5
    Use Sampled Mesh Attribute as Ribbon ID=false
   ) {module=SkeletalMeshLocation001}
   &FN_HermiteSpline(
    Direction End:Vector3f=&Multiply_VectorByFloat(
     Vector=Linked() {linkedVariable {name=Output.SkeletalMeshLocation001.SampledPositionWithoutOffset, type=NiagaraPosition}}
     Float=&Multiply_Float(
      A=-1
      B=&MergeHelper_FloatSwitch(0=$User.GhostLerp1, 1=$User.GhostLerp2, 2=$User.GhostLerp3, 3=?, Index=&SpawnGroup(), Num=::NewEnumerator1)
     )
    )
    Position Blend:NiagaraFloat=$Particles.NormalizedAge
    Position End:Vector3f=&Lerp_Vector(
     A=Linked() {linkedVariable {name=Particles.Previous.Position, type=NiagaraPosition}}
     B=Linked() {linkedVariable {name=Output.SkeletalMeshLocation001.SampledPositionWithoutOffset, type=NiagaraPosition}}
     Alpha=&FloatFromCurve(
      FloatCurve=DataInterface() {propertyValues {Curve {
         defaultValue=1
         keys [
          {interpMode=RCIM_Cubic, leaveTangent=1.6631, tangentMode=RCTM_Break, time=0, value=0.076}
          {interpMode=RCIM_Cubic, tangentMode=RCTM_Break, time=1, value=1}
         ]
        }}}
      Scale Curve=&MergeHelper_FloatSwitch(0=$User.GhostLerp1, 1=$User.GhostLerp2, 2=$User.GhostLerp3, 3=?, Index=&SpawnGroup(), Num=::NewEnumerator1)
     )
    )
    Position Start:Vector3f=&Lerp_Vector(
     A=Linked() {linkedVariable {name=Particles.Initial.SkeletalMeshLocation.SampledPosition, type=NiagaraPosition}}
     B=Linked() {linkedVariable {name=Particles.SkeletalMeshLocation001.SampledPosition, type=NiagaraPosition}}
     Alpha=&MergeHelper_FloatSwitch(0=$User.GhostLerp1, 1=$User.GhostLerp2, 2=$User.GhostLerp3, 3=?, Index=&SpawnGroup(), Num=::NewEnumerator1)
    )
    Write Particles Position=true
   )
   &ParticleState()
   &SolveForcesAndVelocity(Clamp Velocity=false, Limit Acceleration=false)
   &ScaleRibbonWidth(Ribbon Width Scale=&Lerp_Float(
     A=&FloatFromCurve(FloatCurve=DataInterface() {propertyValues {Curve {keys [<at 0 1.5122>,<at 1 0.0686>]}}}, Scale Curve=2)
     B=&FloatFromCurve(
      FloatCurve=DataInterface() {propertyValues {Curve {keys [
          {interpMode=RCIM_Cubic, leaveTangent=-3.8671, tangentMode=RCTM_Break, time=0.4757, value=1}
          {arriveTangent=-1.3032, interpMode=RCIM_Cubic, leaveTangent=-1.3032, tangentMode=RCTM_User, time=0.709, value=0.7882}
          <at 1 0>
         ]}}}
      Scale Curve=2
     )
     Alpha=&MergeHelper_FloatSwitch(0=$User.GhostLerp1, 1=$User.GhostLerp2, 2=$User.GhostLerp3, 3=?, Index=&SpawnGroup(), Num=::NewEnumerator1)
    ))
   &DynamicMaterialParameters(
    Index 0 Param 1=1
    Index 0 Param 4=1
    Module.Index0Float0Write=true
    Module.Index0Float1Write=true
    Module.Index0Float2Write=true
    Module.Index0Float3Write=true
   )
   /SetVariables(
    Particles.AngryColor:LinearColor=&MakeCustomLinearColorFromBool(
     Bool=&SetBoolByFloatComparison(A=$User.GhostLerp1, B=0.5)
     True Color=<rgba 1 1 1 0>
     False Color=<rgba 1 1 1 1>
    )
    Particles.ScaredColor:LinearColor=&MakeCustomLinearColorFromBool(
     Bool=&SetBoolByFloatComparison(A=$User.GhostLerp1, B=0.5)
     True Color=<rgba 1 1 1 1>
     False Color=<rgba 1 1 1 0>
    )
   ) {isSetParameters=true, module=SetVariables_#4}
   &ScaleColor(
    Linear Color Curve=DataInterface() {propertyValues {
      AlphaCurve {keys [<at 0 0>,<at 1 1>]}
      BlueCurve {keys [<at 0 1>,<at 1 1>]}
      GreenCurve {keys [<at 0 1>,<at 1 1>]}
      RedCurve {keys [<at 0 1>,<at 1 1>]}
     }}
    Scale Mode=2
   )
   &CameraOffset(Camera Offset Amount=-5)
  ]
  rendererClasses [&NiagaraRibbonRendererProperties]
  renderers [{rendererClass=&NiagaraRibbonRendererProperties}]
 }
}
summary {
 systemName=NS_Emote_FearCatch_Ghost
 emitters [
  {emitterName=Ghost1, rendererClasses [&NiagaraRibbonRendererProperties]}
  {emitterName=Ghost2, rendererClasses [&NiagaraRibbonRendererProperties]}
  {emitterName=Ghost3, rendererClasses [&NiagaraRibbonRendererProperties]}
  {emitterName=Ghost4, rendererClasses [&NiagaraRibbonRendererProperties]}
 ]
 userVariables [
  User.GhostLerp1 {type=NiagaraFloat, defaultValue=0, description=""}
  User.GhostLerp2 {type=NiagaraFloat, defaultValue=0, description=""}
  User.GhostLerp3 {type=NiagaraFloat, defaultValue=0, description=""}
  User.GhostSpawn1 {type=NiagaraFloat, defaultValue=0, description=""}
  User.GhostSpawn2 {type=NiagaraFloat, defaultValue=0, description=""}
  User.GhostSpawn3 {type=NiagaraFloat, defaultValue=0, description=""}
 ]
}
