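# Niagara system for the ghost beam effect in the FearCatch player emote, built from ribbon beam, ring mesh and flare emitters.
# Beam and rings follow sampled skeletal mesh locations along Hermite splines, with rings adding waveform wobble and lights.
# Curves over User.LoopDur widen, then collapse the beam.
# A delayed ImpactFLare bursts grouped sprites and light, varied by User.RayIndex, while User.ChargePercent tints rings.
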NiagaraSystem /BRCosmetics/Animation/Game/MainPlayer/Emotes/FearCatch/Fx/NS_Emote_FearCatch_GhostBeam
dependencies {
 usedDataInterfaces [
  &NiagaraDataInterfaceCurve
  &NiagaraDataInterfaceColorCurve
  &NiagaraDataInterfaceArrayColor
  &NiagaraDataInterfaceVector2DCurve
  &NiagaraDataInterfaceSpriteRendererInfo
  &NiagaraDataInterfaceSkeletalMesh
 ]
 usedDynamicInputs [
  &MakeVector4_FromLinearColor
  &LerpLinearColors
  &Subtract_Float
 ]
 usedModules [
  &self:SystemSpawnScript.SystemScriptSource.SystemScriptGraph.NiagaraNodeAssignment_1.SetVariables_#1
  &SystemState
  &self:SystemSpawnScript.SystemScriptSource.SystemScriptGraph.NiagaraNodeAssignment_0.SetVariables_#2
 ]
}
emitters {
 BlowingParticles {
  emitterUpdateScript [
   &EmitterState(
    Inactive Response=::NewEnumerator1
    Life Cycle Mode=::NewEnumerator1
    Loop Behavior=::NewEnumerator1
    Loop Duration=$User.LoopDur
    UseLoopDelay=false
   )
   &SpawnRate(SpawnRate=150, Use Spawn Probability=false)
  ]
  particleSpawnScript [
   &InitializeParticle(
    Color=<rgba 0 0.9149 1.5 1>
    Lifetime=0.15
    Mass Max=1.5
    Mass Mode=::NewEnumerator3
    Ribbon Width=11
    Ribbon Width Mode=::NewEnumerator1
    Sprite Rotation Mode=::NewEnumerator2
    UsePositionOffset=false
    Write Material Random=true
   )
   &DynamicMaterialParameters(
    Index 0 Param 1=&Multiply_Float(A=$Emitter.Age, B=-1.0998)
    Index 0 Param 2=0.5747
    Index 0 Param 4=0.6545
    Module.Index0Float0Write=true
    Module.Index0Float1Write=true
    Module.Index0Float2Write=true
    Module.Index0Float3Write=true
    Module.Index1Float0Write=true
    Module.Index1Float1Write=true
    Module.Index1Float2Write=true
    Module.Index1Float3Write=true
    Module.Index2Float0Write=true
    Module.Index2Float1Write=true
    Module.Index2Float2Write=true
    Module.Index2Float3Write=true
    Write Parameter Index 1=::NewEnumerator1
    Write Parameter Index 2=::NewEnumerator1
   )
   &AddVelocity(Velocity=<0 1 0>, Velocity Speed Scale=2860.2427) {enabled=false}
  ]
  particleUpdateScript [
   /SetVariables(Particles.DynPar:Vector4f=&MakeVector4(
     X=1.299
     Y=&FloatFromCurve(
      FloatCurve=DataInterface() {propertyValues {Curve {keys [{interpMode=RCIM_Cubic, leaveTangent=-11.3882, tangentMode=RCTM_Break, time=0.6947, value=0.4229}, <at 0.8842 0>]}}}
      CurveIndex=$Emitter.NormalizedLoopAge
      Scale Curve=2.1776
     )
     Z=0
     W=0.6471
    )) {isSetParameters=true}
   &ParticleState()
   &SkeletalMeshLocation(
    MaskSpawnGroup=false
    Mesh Sampling Type=::NewEnumerator5
    Orientation Sampling=::NewEnumerator1
    Position Sampling=::NewEnumerator2
    Sampled Orientation Coordinate Space=::Local
    Skeletal Mesh=$System.Skeletal Mesh Reader
    Socket Sampling Mode=::NewEnumerator5
   )
   /Game/Effects/Niagara/Modules/Curves/FN_HermiteSpline(
    Position Blend:NiagaraFloat=$Particles.NormalizedAge
    Position End:Vector3f=&AddVectorToPosition(
     Position=$Particles.Previous.Position
     Vector=&RotateVectorByQuaternion(
      Vector To Rotate=&Lerp_Vector(
       A=<0 27 0>
       B=<1 0 1>
       Alpha=&FloatFromCurve(
        FloatCurve=DataInterface() {propertyValues {Curve {defaultValue=1, keys [{tangentMode=RCTM_User, time=0.6985, value=0}, {tangentMode=RCTM_User, time=0.9037, value=1}]}}}
        CurveIndex=$Emitter.NormalizedLoopAge
       )
      )
      Quaternion=$Particles.SkeletalMeshLocation.SampledRotation
     )
    )
    Position Start:NiagaraPosition=$Output.SkeletalMeshLocation.SampledPositionWithoutOffset
    Write Particles Position=true
   )
   &ScaleVelocity(Velocity Scale=&VectorFromFloat(Value=&FloatFromCurve(FloatCurve=DataInterface() {propertyValues {Curve {keys [<at 0 1>,<at 1 0>]}}}, CurveIndex=$Emitter.NormalizedLoopAge))) {enabled=false}
   &SolveForcesAndVelocity(Clamp Velocity=false, Limit Acceleration=false)
   &ScaleRibbonWidth(Ribbon Width Scale=&FloatFromCurve(
     FloatCurve=DataInterface() {propertyValues {Curve {keys [
         {arriveTangent=-1.2235, interpMode=RCIM_Cubic, leaveTangent=-0.1691, tangentMode=RCTM_Break, time=0, value=0.8}
         {arriveTangent=3.4232, interpMode=RCIM_Cubic, leaveTangent=3.4232, time=0.4388, value=1.8374}
         {arriveTangent=10.4288, interpMode=RCIM_Cubic, leaveTangent=10.4288, time=0.7438, value=3.3463}
         {arriveTangent=10.9038, interpMode=RCIM_Cubic, leaveTangent=10.9038, time=0.8532, value=6.1598}
         {arriveTangent=-6.471, interpMode=RCIM_Cubic, tangentMode=RCTM_Break, time=0.9899, value=6.0298}
        ]}}}
     Scale Curve=&FloatFromCurve(
      FloatCurve=DataInterface() {propertyValues {Curve {keys [{interpMode=RCIM_Cubic, leaveTangent=-16.8498, tangentMode=RCTM_Break, time=0.7892, value=1}, <at 1 0>]}}}
      CurveIndex=$Emitter.NormalizedLoopAge
     )
    ))
   &ScaleColor(
    Scale Alpha=&FloatFromCurve(FloatCurve=DataInterface() {propertyValues {Curve {keys [<at 0.8423 1.6557>,<at 1 1.4528>]}}}, CurveIndex=$Emitter.NormalizedLoopAge)
    ScaleA=true
    ScaleRGB=true
   )
   /SetVariables(
    Particles.MediumBeamColor:LinearColor=/Niagara/DynamicInputs/Add/Add_LinearColor(A:LinearColor=$Particles.Color, B:LinearColor=<rgba 0 1.1699 10 1>)
    Particles.SmallerBeamColor:LinearColor=&Add_LinearColor(A:LinearColor=$Particles.Color, B:LinearColor=<rgba 3.4235 5.1352 10 1>)
    Particles.Width:NiagaraFloat=&FloatFromCurve(
     FloatCurve=DataInterface() {propertyValues {Curve {keys [
         {arriveTangent=-1.2235, interpMode=RCIM_Cubic, leaveTangent=-0.1691, tangentMode=RCTM_Break, time=0, value=0.8}
         {arriveTangent=4.1616, interpMode=RCIM_Cubic, leaveTangent=4.1616, time=0.7132, value=0.8818}
         {arriveTangent=17.9228, interpMode=RCIM_Cubic, leaveTangent=17.9228, time=0.8364, value=4.2808}
         {arriveTangent=6.6501, interpMode=RCIM_Cubic, tangentMode=RCTM_Break, time=0.9819, value=5.6966}
        ]}}}
     Scale Curve=&FloatFromCurve(
      FloatCurve=DataInterface() {propertyValues {Curve {keys [{interpMode=RCIM_Cubic, leaveTangent=-19.7783, tangentMode=RCTM_Break, time=0.7633, value=1}, <at 0.8842 0>]}}}
      CurveIndex=$Emitter.NormalizedLoopAge
      Scale Curve=3.8955
     )
    )
    Particles.Width2:NiagaraFloat=&FloatFromCurve(
     FloatCurve=DataInterface() {propertyValues {Curve {keys [
         {arriveTangent=-1.2235, interpMode=RCIM_Cubic, leaveTangent=-0.1691, tangentMode=RCTM_Break, time=0, value=0.8}
         {arriveTangent=9.5701, interpMode=RCIM_Cubic, leaveTangent=9.5701, time=0.7314, value=3.3614}
         {arriveTangent=20.9257, interpMode=RCIM_Cubic, leaveTangent=20.9257, time=0.8404, value=8.8431}
         {arriveTangent=6.6501, interpMode=RCIM_Cubic, tangentMode=RCTM_Break, time=0.9899, value=8.7712}
        ]}}}
     Scale Curve=&FloatFromCurve(
      FloatCurve=DataInterface() {propertyValues {Curve {keys [{interpMode=RCIM_Cubic, leaveTangent=-19.7783, tangentMode=RCTM_Break, time=0.7633, value=1}, <at 0.8842 0>]}}}
      CurveIndex=$Emitter.NormalizedLoopAge
      Scale Curve=4.7062
     )
    )
   ) {isSetParameters=true}
  ]
  rendererClasses [&NiagaraRibbonRendererProperties]
  renderers [
   {rendererClass=&NiagaraRibbonRendererProperties}
   {rendererClass=&NiagaraRibbonRendererProperties, rendererIndex=1}
   {rendererClass=&NiagaraRibbonRendererProperties, rendererIndex=2}
  ]
 }
 ImpactFLare {
  emitterUpdateScript [
   &EmitterState(
    Inactive Response=::NewEnumerator1
    Life Cycle Mode=::NewEnumerator1
    Loop Behavior=::NewEnumerator1
    Loop Delay=$System.float
    Loop Duration=1
    UseLoopDelay=true
   )
   &SpawnBurst_Instantaneous(Spawn Group=1, Use Spawn Probability=false, UseLoopCountLimit=false) {module=SpawnBurst_Instantaneous001}
   &SpawnBurst_Instantaneous(Use Spawn Probability=false, UseLoopCountLimit=false)
   &SpawnBurst_Instantaneous(Spawn Count=12, Spawn Group=2, Use Spawn Probability=false, UseLoopCountLimit=false) {module=SpawnBurst_Instantaneous002}
   &SpawnBurst_Instantaneous(
    Spawn Count=&MergeHelper_IntSwitch(
     0=&MergeHelper_IntSwitch(0=13, 1=17, 2=30, 3=30, 4=30, 5=?, Index=$Engine.QualityLevel, Num=::NewEnumerator3)
     1=&MergeHelper_IntSwitch(0=13, 1=17, 2=30, 3=30, 4=30, 5=?, Index=$Engine.QualityLevel, Num=::NewEnumerator3)
     2=0
     3=?
     4=?
     5=?
     Index=&MakeIntFromFloat(Float=$User.RayIndex)
     Num=::NewEnumerator1
    )
    Spawn Group=3
    Use Spawn Probability=false
    UseLoopCountLimit=false
   ) {enabled=false, module=SpawnBurst_Instantaneous003}
  ]
  particleSpawnScript [
   /SetVariables(Particles.VisibilityTag=&SpawnGroup()) {isSetParameters=true}
   &InitializeParticle(
    Color=&MergeHelper_ColorSwitch(
     0=<rgba 0.013 1.5 1.1172 1>
     1=<rgba 0 3 2.4954 1>
     2=&SelectLinearColorFromArray(
      Array Sampling Mode=::NewEnumerator0
      Direct Array Index=?
      Color Selection Array=DataInterface() {propertyValues {ColorData [<rgba 0 3 1.5002 1>, <rgba 0 1.0208 2 1>]}}
      Random Seed=0
     )
     3=&SelectLinearColorFromArray(
      Array Sampling Mode=::NewEnumerator0
      Direct Array Index=?
      Color Selection Array=DataInterface() {propertyValues {ColorData [<rgba 0 3 1.5002 1>, <rgba 0.0251 0.2834 1 1>]}}
      Random Seed=0
     )
     Index=&SpawnGroup()
     Num=::NewEnumerator2
    )
    Lifetime=&MergeHelper_FloatSwitch(
     0=0.5
     1=&MaskFloat(
      Mask Source=1
      Mask Parameter=&MakeIntFromFloat(Float=$User.RayIndex)
      Mask=2
      Masked Float=0.5
      Passthrough Float=0.2
     )
     2=&RandomRangeFloat(Minimum=0.1, Maximum=0.2)
     3=&RandomRangeFloat(Minimum=0.8, Maximum=0.6)
     4=0.88
     Index=&SpawnGroup()
     Num=::NewEnumerator3
    )
    Position Offset Coordinate Space=::Local
    Sprite Rotation Mode=::NewEnumerator2
    Sprite Size Mode=::NewEnumerator3
    UsePositionOffset=true
    Write Material Random=true
   )
   &CameraOffset(Camera Offset Amount=30)
   &ShapeLocation(MaskSpawnGroup=true, Override Seed=false, Spawn Group Mask=3, Sphere Radius=9, UseOwnerScale=true)
   &AddVelocity(
    Module.Use Falloff Distance=false
    Override Seed=false
    Velocity Mode=::NewEnumerator1
    Velocity Speed=&MaskFloatBySpawnGroup(Masked Float=6.5157, Passthrough Float=0, Spawn Group Mask=2)
   )
   &AddVelocity(
    Module.Use Falloff Distance=false
    Override Seed=false
    Velocity Mode=::NewEnumerator1
    Velocity Origin=$Output.SocketLocation001.SampledPosition
    Velocity Speed=&MergeHelper_FloatSwitch(0=0, 1=0, 2=0, 3=450, 4=450, Index=&SpawnGroup(), Num=::NewEnumerator3)
   ) {module=AddVelocity001}
   /SetVariables(Particles.StreakAlign:Vector3f=$Particles.ShapeLocation.ShapeNormal) {isSetParameters=true}
  ]
  particleUpdateScript [
   &SkeletalMeshLocation(MaskSpawnGroup=false, Mesh Sampling Type=::NewEnumerator5, Skeletal Mesh=$System.Skeletal Mesh Reader, Socket Sampling Mode=::NewEnumerator5) {module=SkeletalMeshLocation001}
   &ShapeLocation(MaskSpawnGroup=true, Override Seed=false, Spawn Group Mask=4, Sphere Radius=6.91, Sphere Surface Distribution=1, UseOwnerScale=true) {module=ShapeLocation001}
   &Drag(
    Drag=&FloatFromCurve(
     FloatCurve=DataInterface() {propertyValues {Curve {
        defaultValue=1
        keys [
         {arriveTangent=4.1785, interpMode=RCIM_Cubic, leaveTangent=4.1785, tangentMode=RCTM_User, time=0, value=0}
         {arriveTangent=1.1179, interpMode=RCIM_Cubic, leaveTangent=1.1179, tangentMode=RCTM_User, time=0.2042, value=0.6228}
         {arriveTangent=0.0224, interpMode=RCIM_Cubic, leaveTangent=0.0224, tangentMode=RCTM_User, time=1, value=1}
        ]
       }}}
     Scale Curve=&RandomRangeFloat(Minimum=14.0132, Maximum=19.8906)
    )
    Use Linear Drag=true
    Use Rotational Drag=true
   )
   &InheritSourceMovement(ApplyForce=false, ApplyPosition=true, ApplyPositionFalloff=false)
   &CurlNoiseForce(
    Bias Noise=false
    Noise Frequency=27.7223
    Noise Strength=&MaskFloatBySpawnGroup(
     Masked Float=&FloatFromCurve(
      FloatCurve=DataInterface() {propertyValues {Curve {
         defaultValue=1
         keys [
          {arriveTangent=4.1785, interpMode=RCIM_Cubic, leaveTangent=4.1785, tangentMode=RCTM_User, time=0, value=0}
          {arriveTangent=1.1179, interpMode=RCIM_Cubic, leaveTangent=1.1179, tangentMode=RCTM_User, time=0.2042, value=0.6228}
          {arriveTangent=0.0224, interpMode=RCIM_Cubic, leaveTangent=0.0224, tangentMode=RCTM_User, time=1, value=1}
         ]
        }}}
      Scale Curve=500
     )
     Passthrough Float=0
     Spawn Group Mask=3
    )
    Pan Noise=false
   )
   &SolveForcesAndVelocity(Clamp Velocity=false, Limit Acceleration=false)
   &ParticleState()
   &ScaleSpriteSize(
    Scale Sprite Size Mode=::NewEnumerator0
    Uniform Scale Factor=&MergeHelper_FloatSwitch(
     0=&MakeFloatFromVector2d(Vector 2D=&Vector2DFromFloat(Value=&FloatFromCurve(FloatCurve=DataInterface() {propertyValues {Curve {
           defaultValue=1
           keys [
            {interpMode=RCIM_Cubic, leaveTangent=4.354, tangentMode=RCTM_Break, time=0, value=0.2}
            {arriveTangent=0.8588, interpMode=RCIM_Cubic, leaveTangent=0.8588, tangentMode=RCTM_User, time=0.2851, value=1.6226}
            {arriveTangent=-0.4169, tangentMode=RCTM_User, time=1, value=2}
           ]
          }}})))
     1=&MakeFloatFromVector2d(Vector 2D=&Vector2DFromFloat(Value=&FloatFromCurve(FloatCurve=DataInterface() {propertyValues {Curve {
           defaultValue=1
           keys [
            {interpMode=RCIM_Cubic, leaveTangent=4.354, tangentMode=RCTM_Break, time=0, value=0.2}
            {arriveTangent=0.8588, interpMode=RCIM_Cubic, leaveTangent=0.8588, tangentMode=RCTM_User, time=0.2851, value=1.6226}
            {arriveTangent=-0.4169, tangentMode=RCTM_User, time=1, value=2}
           ]
          }}})))
     2=&FloatFromCurve(FloatCurve=DataInterface() {propertyValues {Curve {keys [<at 0 1>,<at 1 0>]}}})
     3=?
     Index=&SpawnGroup()
     Num=::NewEnumerator1
    )
   )
   /SetVariables() {isSetParameters=true}
   /SetVariables(Particles.GlowSize=&Vector2DFromFloat(Value=&FloatFromCurve(
      FloatCurve=DataInterface() {propertyValues {Curve {
         defaultValue=1
         keys [
          {interpMode=RCIM_Cubic, leaveTangent=4.354, tangentMode=RCTM_Break, time=0, value=0.2}
          {arriveTangent=-6.293, interpMode=RCIM_Cubic, leaveTangent=-6.293, tangentMode=RCTM_User, time=0.2851, value=1.6226}
          {arriveTangent=-0.4169, tangentMode=RCTM_User, time=0.7722, value=0}
         ]
        }}}
      Scale Curve=35.6667
     ))) {isSetParameters=true}
   /SetVariables(Particles.FlareSize:Vector2f=/Niagara/DynamicInputs/Helpers/MaskVector2d(
     Mask Source:NiagaraInt32=1
     Mask Parameter:NiagaraInt32=&MakeIntFromFloat(Float=$User.RayIndex)
     Mask:NiagaraInt32=2
     Masked Vector2D:Vector2f=&Vector2DFromFloat(Value=&FloatFromCurve(
       FloatCurve=DataInterface() {propertyValues {Curve {
          defaultValue=1
          keys [
           {interpMode=RCIM_Cubic, leaveTangent=3.8251, tangentMode=RCTM_Break, time=0, value=0.2}
           {arriveTangent=1.1979, interpMode=RCIM_Cubic, leaveTangent=1.1979, tangentMode=RCTM_User, time=0.2391, value=2.0725}
           {arriveTangent=-0.0483, interpMode=RCIM_Cubic, leaveTangent=-0.0483, tangentMode=RCTM_User, time=0.5684, value=2.1488}
           {arriveTangent=-13.9709, interpMode=RCIM_Cubic, leaveTangent=-13.9709, tangentMode=RCTM_User, time=0.6204, value=1.7462}
           {arriveTangent=-0.4169, tangentMode=RCTM_User, time=0.9333, value=0}
          ]
         }}}
       Scale Curve=276.2582
      ))
     Passthrough Vector2D:Vector2f=&Vector2DFromFloat(Value=&FloatFromCurve(
       FloatCurve=DataInterface() {propertyValues {Curve {
          defaultValue=1
          keys [
           {interpMode=RCIM_Cubic, leaveTangent=4.354, tangentMode=RCTM_Break, time=0, value=0.2}
           {arriveTangent=0.8588, interpMode=RCIM_Cubic, leaveTangent=0.8588, tangentMode=RCTM_User, time=0.2216, value=2.0725}
           {arriveTangent=-0.4169, tangentMode=RCTM_User, time=0.9924, value=0.2003}
          ]
         }}}
       Scale Curve=76.8572
      ))
    )) {isSetParameters=true}
   /SetVariables(Particles.StreakSize:Vector2f=&Vector2DFromCurve(
     Vector2Curve=DataInterface() {propertyValues {XCurve {keys [<at 0.5 1>,<at 1 0>]}, YCurve {keys [<at 0 0.2>,<at 0.5 1>]}}}
     ScaleCurve=<2.0313 47.5964>
    )) {isSetParameters=true}
   /SetVariables(
    Particles.DynParam1:Vector4f=<1 0 0 0>
    Particles.ParticleSize:Vector2f=&Multiply_Vector2D_ByFloat(
     Float=&FloatFromCurve(FloatCurve=DataInterface() {propertyValues {Curve {keys [<at 0.4473 1>,<at 1 0>]}}})
     Vector2D=&MakeVector2D(
      X=1
      Y=&Multiply_Float(A=&VectorLength(Vector=$Particles.Velocity), B=0.1)
     )
    )
   ) {isSetParameters=true}
   &DynamicMaterialParameters(
    Index 0 Param 1=6.0238
    Index 0 Param 2=1
    Index 0 Param 3=0.3113
    Index 0 Param 4=1
    Module.Index0Float0Write=true
    Module.Index0Float1Write=true
    Module.Index0Float2Write=true
    Module.Index0Float3Write=true
    Module.Index1Float0Write=true
    Module.Index1Float1Write=true
    Module.Index1Float2Write=true
    Module.Index1Float3Write=true
    Write Parameter Index 1=::NewEnumerator1
   )
   &FN_Light_Attributes(
    Color=&ColorFromCurve(
     ColorCurve=DataInterface() {propertyValues {
       AlphaCurve {keys [<at 0 1 RCIM_Cubic>,<at 1 0 RCIM_Cubic>]}
       BlueCurve {keys [<at 0 0.6992 RCIM_Cubic>,<at 0.9874 0>]}
       GreenCurve {keys [<at 0 1.5 RCIM_Cubic>,<at 0.9874 0>]}
       RedCurve {keys [<at 0 0 RCIM_Cubic>,<at 0.9874 0>]}
      }}
     CurveIndex=$Emitter.NormalizedLoopAge
    )
    Create Color Attribute=true
    Create Radius Attribute=true
    Radius=35
   )
   &SubUVAnimation(
    Evaluation Type=::NewEnumerator0
    "Sprite Renderer "=DataInterface() {propertyValues {SpriteRenderer=&self:ImpactFLare_0.Renderer}}
    SubUV Animation Mode:ENiagara_SubUVLookupModeV2=::NewEnumerator2
    UseEndFrame=false
    UseStartFrame=false
   )
  ]
  rendererClasses [&NiagaraSpriteRendererProperties, &NiagaraLightRendererProperties]
  renderers [
   {rendererClass=&NiagaraSpriteRendererProperties}
   {rendererClass=&NiagaraSpriteRendererProperties, rendererIndex=1}
   {rendererClass=&NiagaraLightRendererProperties, rendererIndex=2}
   {rendererClass=&NiagaraSpriteRendererProperties, rendererIndex=3}
   {rendererClass=&NiagaraSpriteRendererProperties, rendererIndex=4}
  ]
 }
 RingsMeshes {
  emitterUpdateScript [
   &EmitterState(
    Inactive Response=::NewEnumerator1
    Life Cycle Mode=::NewEnumerator1
    Loop Behavior=::NewEnumerator1
    Loop Duration=$User.LoopDur
    UseLoopDelay=false
   )
   &SpawnBurst_Instantaneous(Spawn Group=1, Use Spawn Probability=false, UseLoopCountLimit=false)
   &SpawnRate(SpawnRate=12.2736, Use Spawn Probability=false)
  ]
  particleSpawnScript [
   /SetVariables(Particles.VisibilityTag=&SpawnGroup()) {isSetParameters=true}
   &InitializeParticle(
    Color=<rgba 1 1 1 10>
    Lifetime=&MergeHelper_FloatSwitch(0=0.3, 1=$User.LoopDur, 2=?, 3=?, Index=&SpawnGroup(), Num=::NewEnumerator0)
    Mesh Scale Mode=::NewEnumerator3
    Mesh Uniform Scale=&Multiply_Float(A=&MaskFloat(Mask Source=0, Mask Parameter=?, Mask=0, Masked Float=0.3, Passthrough Float=0.3), B=1.5226)
    Sprite Size Mode=::NewEnumerator3
    Uniform Sprite Size=&MaskFloat(Mask Source=0, Mask Parameter=?, Mask=0, Masked Float=50, Passthrough Float=100)
    UsePositionOffset=false
    Write Material Random=true
   )
   &InitialMeshOrientation(Mesh Axis To Orient=<0 0 1>, Mesh Orientation Mode=::NewEnumerator2, Orientation Vector=<0 1 0>, Use Rotation Vector=false)
  ]
  particleUpdateScript [
   &ParticleState()
   &SkeletalMeshLocation(
    MaskSpawnGroup=false
    Mesh Sampling Type=::NewEnumerator5
    Position Sampling=::NewEnumerator2
    Skeletal Mesh=$System.Skeletal Mesh Reader
    Socket Sampling Mode=::NewEnumerator5
   )
   &FN_HermiteSpline(
    Position Blend:NiagaraFloat=$Particles.NormalizedAge
    Position End:Vector3f=&AddVectorToPosition(
     Position=$Particles.Previous.Position
     Vector=&RotateVectorByQuaternion(Vector To Rotate=<0 10.7278 0>, Quaternion=$Particles.SkeletalMeshLocation.SampledRotation)
    )
    Position Start:NiagaraPosition=$Output.SkeletalMeshLocation.SampledPositionWithoutOffset
   )
   /SetVariables(Particles.ParticleDistance:NiagaraFloat=&Multiply_Float(
     A=&Abs_Float(Float:NiagaraFloat=&VectorLength(Vector=Linked() {linkedVariable {name=Particles.Position, type=NiagaraPosition}}))
     B=0.1
    )) {isSetParameters=true}
   /SetVariables(Particles.Position=&Add_Vector(
     A=$Output.FN_HermiteSpline.Position
     B=&Multiply_VectorByFloat(
      Vector=&RotateVectorByQuaternion(
       Vector To Rotate=&MakeVector(
        X=0
        Y=&Waveform(
         Global Input Position=&Multiply_Float(A=$Particles.ParticleDistance, B=0.005)
         Global Amplitude Scale=10
         "Waveform [1]"=::NewEnumerator0
         "[1] Frequency"=1
         "[1] Amplitude Min/Max"=<0 1>
         "[1] Bias"=-0.5
         "[2] Phase"=$Emitter.Age
         "[3] Phase"=$Emitter.Age
        )
        Z=&Waveform(
         Global Input Position=&Multiply_Float(A=$Particles.ParticleDistance, B=0.005)
         Global Amplitude Scale=10
         "Waveform [1]"=::NewEnumerator1
         "[1] Frequency"=1
         "[1] Amplitude Min/Max"=<0 1>
         "[1] Bias"=-0.5
         "[2] Phase"=$Emitter.Age
         "[3] Phase"=$Emitter.Age
        )
       )
       Quaternion=<0 0 0 1>
      )
      Float=$Particles.WobbleOverLength
     )
    )) {isSetParameters=true}
   /SetVariables(Particles.SizeCurve:NiagaraFloat=&FloatFromCurve(
     FloatCurve=DataInterface() {propertyValues {Curve {
        defaultValue=0
        keys [
         {arriveTangent=2.9246, interpMode=RCIM_Cubic, leaveTangent=2.9246, tangentMode=RCTM_User, time=0, value=0.5131}
         {arriveTangent=3.1812, interpMode=RCIM_Cubic, leaveTangent=3.1812, time=0.4842, value=1.65}
         {arriveTangent=9.57, interpMode=RCIM_Cubic, leaveTangent=9.57, time=0.7106, value=2.7736}
         {arriveTangent=9.5416, interpMode=RCIM_Cubic, leaveTangent=9.5416, time=0.8096, value=4.7637}
         {arriveTangent=1.6155, interpMode=RCIM_Cubic, leaveTangent=1.6155, tangentMode=RCTM_User, time=1, value=5.5351}
        ]
       }}}
     Scale Curve=&MakeCustomFloatFromBool(
      Bool=&SetBoolByFloatComparison(A=$User.LoopDur, B=0.65, Comparison Type=::NewEnumerator4)
      True Float=&FloatFromCurve(FloatCurve=DataInterface() {propertyValues {Curve {keys [<at 0.6483 1>,<at 0.7901 0>]}}}, CurveIndex=$Emitter.NormalizedLoopAge)
      False Float=&FloatFromCurve(FloatCurve=DataInterface() {propertyValues {Curve {keys [<at 0.6483 1>,<at 0.8787 0>]}}}, CurveIndex=$Emitter.NormalizedLoopAge)
     )
    )) {isSetParameters=true}
   &ScaleMeshSize(Scale Factor=&MakeVector(X=2.2282, Y=$Particles.SizeCurve, Z=$Particles.SizeCurve))
   &ScaleSpriteSize(
    Uniform Curve Scale=&Lerp_Float(A=0.5, B=1, Alpha=$User.ChargePercent)
    Uniform Curve Sprite Scale=DataInterface() {propertyValues {Curve {
       defaultValue=0
       keys [
        {arriveTangent=6.3581, interpMode=RCIM_Cubic, time=0, value=0}
        {arriveTangent=30.877, tangentMode=RCTM_User, time=0.1, value=1}
        {arriveTangent=-23.2322, interpMode=RCIM_Cubic, leaveTangent=-23.2322, tangentMode=RCTM_User, time=0.9, value=1}
        <at 1 0 RCIM_Cubic>
       ]
      }}}
   )
   /SetVariables(Particles.Soft Glow Color:LinearColor=&ColorFromCurve(ColorCurve=DataInterface() {propertyValues {
       AlphaCurve {keys [<at 0 0.386 RCIM_Cubic>]}
       BlueCurve {keys [<at 0 2 RCIM_Cubic>,<at 1 0.2448 RCIM_Cubic>]}
       GreenCurve {keys [<at 0 1.1499 RCIM_Cubic>,<at 1 0.075 RCIM_Cubic>]}
       RedCurve {keys [<at 0 0 RCIM_Cubic>,<at 1 0.0181 RCIM_Cubic>]}
      }})) {isSetParameters=true}
   &ScaleColor(
    Scale Alpha=0.0779
    Scale RGB=&MakeVectorFromLinearColorRGB(Color=&ColorFromCurve(
      ColorCurve=DataInterface() {propertyValues {
        AlphaCurve {keys [<at 0 1 RCIM_Cubic>]}
        BlueCurve {keys [<at 0 3 RCIM_Cubic>,<at 1 1.3229 RCIM_Cubic>]}
        GreenCurve {keys [<at 0 1.643 RCIM_Cubic>,<at 1 0.7799 RCIM_Cubic>]}
        RedCurve {keys [<at 0 0 RCIM_Cubic>,<at 1 0 RCIM_Cubic>]}
       }}
      CurveIndex=$User.ChargePercent
     ))
    ScaleA=true
    ScaleRGB=true
   )
   /SetVariables(
    Particles.InnerRingColor:LinearColor=&ColorFromCurve(ColorCurve=DataInterface() {propertyValues {
       AlphaCurve {keys [<at 0 0.978 RCIM_Cubic>]}
       BlueCurve {keys [<at 0 17 RCIM_Cubic>]}
       GreenCurve {keys [<at 0 7.6064 RCIM_Cubic>]}
       RedCurve {keys [<at 0 4.6927 RCIM_Cubic>]}
      }})
    Particles.RingInnerSize:Vector3f=&MakeVector(
     X=&Multiply_Float(A=$Particles.SizeCurve, B=0.25)
     Y=&Multiply_Float(A=$Particles.SizeCurve, B=0.25)
     Z=0.148
    )
   ) {isSetParameters=true}
   &FN_Light_Attributes(
    Color=&ColorFromCurve(
     ColorCurve=DataInterface() {propertyValues {
       AlphaCurve {keys [<at 0 1 RCIM_Cubic>,<at 1 0 RCIM_Cubic>]}
       BlueCurve {keys [<at 0 2.7968 RCIM_Cubic>,<at 0.9874 0>]}
       GreenCurve {keys [<at 0 6 RCIM_Cubic>,<at 0.9874 0>]}
       RedCurve {keys [<at 0 0 RCIM_Cubic>,<at 0.9874 0>]}
      }}
     CurveIndex=$User.LoopDur
    )
    Create Color Attribute=true
    Create Position Attribute=true
    Create Radius Attribute=true
    Position:NiagaraPosition=$Output.SkeletalMeshLocation.SampledPositionWithoutOffset
    Radius=134.2071
   )
   &DynamicMaterialParameters(
    Index 0 Param 1=-0.5588
    Module.Index0Float0Write=true
    Module.Index0Float1Write=true
    Module.Index0Float2Write=true
    Module.Index0Float3Write=true
   )
  ]
  rendererClasses [&NiagaraMeshRendererProperties, &NiagaraLightRendererProperties]
  renderers [
   {rendererClass=&NiagaraMeshRendererProperties, rendererIndex=0}
   {rendererClass=&NiagaraMeshRendererProperties, rendererIndex=1}
   {rendererClass=&NiagaraLightRendererProperties, rendererIndex=2}
   {rendererClass=&NiagaraMeshRendererProperties, rendererIndex=3}
  ]
 }
}
summary {
 systemName=NS_Emote_FearCatch_GhostBeam
 emitters [
  {emitterName=BlowingParticles, rendererClasses [&NiagaraRibbonRendererProperties]}
  {emitterName=RingsMeshes, rendererClasses [&NiagaraMeshRendererProperties, &NiagaraLightRendererProperties]}
  {emitterName=ImpactFLare, rendererClasses [&NiagaraSpriteRendererProperties, &NiagaraLightRendererProperties]}
 ]
 userVariables [
  User.BeamEnd_0 {type=NiagaraPosition, description="", defaultValue {}}
  User.ChargePercent {type=NiagaraFloat, defaultValue=0, description=""}
  User.LoopDur {type=NiagaraFloat, defaultValue=2, description=""}
  User.RayIndex {type=NiagaraFloat, defaultValue=0, description=""}
 ]
}
