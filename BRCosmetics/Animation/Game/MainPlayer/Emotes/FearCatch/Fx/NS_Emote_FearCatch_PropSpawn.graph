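# Niagara system for the prop spawn effect in the FearCatch player emote.
# It holds five emitters: airglow, sparkles, toon impact smoke, toon smoke, and a ToonSmoke_LowMedium variant apparently meant for lower quality levels.
# System-level scripts use SystemState and two set-variable assignments, with time-of-day color and float-array selection dynamic inputs among the dependencies.
# No user parameters are exposed.
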NiagaraSystem /BRCosmetics/Animation/Game/MainPlayer/Emotes/FearCatch/Fx/NS_Emote_FearCatch_PropSpawn
dependencies {
 usedDataInterfaces [&NiagaraDataInterfaceArrayFloat]
 usedDynamicInputs [&FNTimeOfDayColor, &SelectFloatFromArray]
 usedModules [
  &self:SystemSpawnScript.SystemScriptSource.SystemScriptGraph.NiagaraNodeAssignment_1.SetVariables_#1
  &self:SystemSpawnScript.SystemScriptSource.SystemScriptGraph.NiagaraNodeAssignment_0.SetVariables_#2
  &SystemState
 ]
}
emitters {
 AirglowLWE002 {rendererClasses [], renderers []}
 SparklesLWE002 {rendererClasses [], renderers []}
 ToonImpactSmoke {rendererClasses [], renderers []}
 ToonSmoke {rendererClasses [], renderers []}
 ToonSmoke_LowMedium {rendererClasses [], renderers []}
}
summary {
 systemName=NS_Emote_FearCatch_PropSpawn
 emitters [
  {emitterName=AirglowLWE002, rendererClasses []}
  {emitterName=SparklesLWE002, rendererClasses []}
  {emitterName=ToonImpactSmoke, rendererClasses []}
  {emitterName=ToonSmoke, rendererClasses []}
  {emitterName=ToonSmoke_LowMedium, rendererClasses []}
 ]
 userVariables []
}
