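# Niagara system for the FearCatch emote's gadget thrusters, drawing green exhaust, glows, wake and smoke.
# LargeWake meshes and thruster exhaust crossfade spawn rates by smoothed owner-velocity thresholds: wake when slow, exhaust when fast.
# Exhaust and pulsing glow sprites, both with lights, track speeder sockets on the gadget's skeletal mesh; smoke follows the speeder bone.
# Smoke's spawn rate references InterpolateSpeedThresholdEmitter_01, although its own threshold module is named _03.
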NiagaraSystem /BRCosmetics/Animation/Game/MainPlayer/Emotes/FearCatch/Fx/NS_Emote_FearCatch_Thrusters
dependencies {
 usedDataInterfaces [
  &NiagaraDataInterfaceSkeletalMesh
  &NiagaraDataInterfaceCurve
  &NiagaraDataInterfaceColorCurve
  &NiagaraDataInterfaceArrayInt32
  &NiagaraDataInterfaceVectorCurve
 ]
 usedDynamicInputs []
 usedModules [&SystemState]
}
emitters {
 LargeWake {
  emitterUpdateScript [
   &EmitterState(
    Delay First Loop Only=true
    Inactive Response=::NewEnumerator1
    Life Cycle Mode=::NewEnumerator1
    Loop Delay=0.4
    Loop Duration=1
    UseLoopDelay=true
   )
   &self:InterpolateSpeedThresholdEmitter_01(
    Define Initial Value=false
    Rate Of Change=5
    Target Float Value:NiagaraFloat=&MakeCustomFloatFromBool(
     Bool=&SetBoolByFloatComparison(
      A=&Add_Float(
       A=&Add_Float(
        A=&Abs_Float(Float:NiagaraFloat=&MakeFloatFromVector(Vector=$Engine.Owner.Velocity, Channel=::NewEnumerator0))
        B=&Abs_Float(Float:NiagaraFloat=&MakeFloatFromVector(Vector=$Engine.Owner.Velocity, Channel=::NewEnumerator2))
       )
       B=&Abs_Float(Float:NiagaraFloat=&MakeFloatFromVector(Vector=$Engine.Owner.Velocity, Channel=::NewEnumerator1))
      )
      B=46.278
     )
     True Float=1
     False Float=0
    )
   ) {module=InterpolateSpeedThresholdEmitter_01}
   &SpawnRate(
    SpawnRate=&Lerp_Float(A=4.5, B=0, Alpha=$Emitter.InterpolateSpeedThresholdEmitter_01.InterpolateOverTime_Float.Moving Average)
    Use Spawn Probability=false
   )
  ]
  particleSpawnScript [
   &InitializeParticle(
    Color=<rgba 0.5634 1 0.5143 0.2348>
    Lifetime Max=1
    Lifetime Min=0.8
    Lifetime Mode=::NewEnumerator1
    Mesh Scale=&Multiply_VectorByFloat(
     Vector=&RandomRangeVector(Minimum=<1 1 0.5>, Maximum=<1 1 1>)
     Float=&RandomRangeFloat(Minimum=0.65, Maximum=1.1)
    )
    Mesh Scale Mode=::NewEnumerator4
    Position Offset=&MakeVector(X=0, Y=0, Z=-80)
    UsePositionOffset=true
    Write Material Random=true
   )
   &InitialMeshOrientation(
    Mesh Orientation Mode=::NewEnumerator2
    Orientation Coordinate Space=::World
    Rotation=&MakeVector(
     X=&MakeCustomFloatFromBool(Bool=&RandomBool(), True Float=0, False Float=0)
     Y=&MakeCustomFloatFromBool(Bool=&RandomBool(), True Float=0, False Float=0)
     Z=&RandomRangeFloat(Minimum=-1, Maximum=1)
    )
    Use Rotation Vector=true
   )
  ]
  particleUpdateScript [
   &ParticleState()
   &DynamicMaterialParameters(
    Index 0 Param 1=&FloatFromCurve(FloatCurve=DataInterface() {propertyValues {Curve {keys [
         <at 0 0.2 RCIM_Cubic>
         {interpMode=RCIM_Cubic, tangentMode=RCTM_User, time=0.3, value=0.45}
         {interpMode=RCIM_Cubic, tangentMode=RCTM_User, time=1, value=0}
        ]}}})
    Module.Param0WriteEnabled=true
    Module.Param1WriteEnabled=true
    Module.Param2WriteEnabled=true
    Module.Param3WriteEnabled=true
   )
   &ScaleMeshSize(Scale Factor=&Multiply_VectorByFloat(
     Vector=&MakeVector(
      X=1
      Y=1
      Z=&FloatFromCurve(FloatCurve=DataInterface() {propertyValues {Curve {keys [<at 0.15 1>,<at 0.6 0>]}}})
     )
     Float=&FloatFromCurve(
      FloatCurve=DataInterface() {propertyValues {Curve {keys [
          {arriveTangent=2.3149, interpMode=RCIM_Cubic, leaveTangent=2.3149, tangentMode=RCTM_User, time=0, value=0}
          {arriveTangent=0.9995, interpMode=RCIM_Cubic, leaveTangent=0.9995, tangentMode=RCTM_User, time=0.3, value=0.45}
          {arriveTangent=-0.0405, interpMode=RCIM_Cubic, leaveTangent=-0.0405, tangentMode=RCTM_User, time=1, value=1}
         ]}}}
      Scale Curve=2.5
     )
    ))
  ]
  rendererClasses [&NiagaraMeshRendererProperties]
  renderers [{rendererClass=&NiagaraMeshRendererProperties, rendererIndex=0}]
 }
 LargeWake001 {
  emitterUpdateScript [
   &EmitterState(
    Delay First Loop Only=true
    Inactive Response=::NewEnumerator1
    Life Cycle Mode=::NewEnumerator1
    Loop Delay=0.4
    Loop Duration=1
    UseLoopDelay=true
   )
   &self:InterpolateSpeedThresholdEmitter_01(
    Define Initial Value=false
    Rate Of Change=5
    Target Float Value:NiagaraFloat=&MakeCustomFloatFromBool(
     Bool=&SetBoolByFloatComparison(
      A=&Add_Float(
       A=&Add_Float(
        A=&Abs_Float(Float:NiagaraFloat=&MakeFloatFromVector(Vector=$Engine.Owner.Velocity, Channel=::NewEnumerator0))
        B=&Abs_Float(Float:NiagaraFloat=&MakeFloatFromVector(Vector=$Engine.Owner.Velocity, Channel=::NewEnumerator2))
       )
       B=&Abs_Float(Float:NiagaraFloat=&MakeFloatFromVector(Vector=$Engine.Owner.Velocity, Channel=::NewEnumerator1))
      )
      B=147.0222
     )
     True Float=1
     False Float=0
    )
   ) {module=InterpolateSpeedThresholdEmitter_01}
   &SpawnBurst_Instantaneous(Spawn Count=3, Spawn Group=3, Use Spawn Probability=false, UseLoopCountLimit=false)
   &SpawnRate(
    SpawnRate=&Lerp_Float(A=0, B=30, Alpha=$Emitter.InterpolateSpeedThresholdEmitter_01.InterpolateOverTime_Float.Moving Average)
    Use Spawn Probability=false
   )
   &SpawnRate(
    SpawnGroup=2
    SpawnRate=&Lerp_Float(A=0, B=9, Alpha=$Emitter.InterpolateSpeedThresholdEmitter_01.InterpolateOverTime_Float.Moving Average)
    Use Spawn Probability=false
   ) {module=SpawnRate002}
   &SpawnBurst_Instantaneous(Use Spawn Probability=false, UseLoopCountLimit=false) {enabled=false, module=SpawnBurst_Instantaneous001}
   &SpawnRate(
    SpawnGroup=1
    SpawnRate=&Lerp_Float(A=41.2082, B=15, Alpha=$Emitter.InterpolateSpeedThresholdEmitter_01.InterpolateOverTime_Float.Moving Average)
    Use Spawn Probability=false
   ) {enabled=false, module=SpawnRate001}
  ]
  particleSpawnScript [
   /SetVariables(Particles.VisibilityTag=&SpawnGroup()) {isSetParameters=true, module=SetVariables_#1}
   &InitializeParticle(
    Color=<rgba 0.413 3 0 0.4348>
    Lifetime=0.3
    Mesh Scale Mode=::NewEnumerator4
    Position Offset=&MakeVector(X=0, Y=0, Z=-80)
    Sprite Size Mode=::NewEnumerator3
    Uniform Sprite Size=28.2139
    UsePositionOffset=true
    Write Material Random=true
   )
   &InitialMeshOrientation(Mesh Orientation Mode=::NewEnumerator2, Orientation Coordinate Space=::World, Orientation Vector=<0 1 0>, Use Rotation Vector=false)
   &PartitionParticles(Number Of Partitions=3)
  ]
  particleUpdateScript [
   &ParticleState(Kill Particles When Lifetime Has Elapsed=false)
   &SkeletalMeshLocation(
    Bone / Socket Index=&SelectIntFromArray(Direct Array Index=$Particles.PartitionParticles.Partition, Int Selection Array=DataInterface() {propertyValues {IntData [0,1,2]}})
    MaskSpawnGroup=false
    Mesh Sampling Type=::NewEnumerator5
    Skeletal Mesh=DataInterface() {propertyValues {
      DefaultMesh=null
      ExcludeBoneName=root
      OptionalComponentClass=null
      PreviewMesh=/BRCosmetics/Animation/Props/Gadgets_FearCatch/Meshes/FearCatch_Gadget
      SkinningMode=SkinOnTheFly
      SoftSourceActor=null
      SourceMode=Default
      UvSetIndex=0
      WholeMeshLOD=-1
      bExcludeBone=false
      bReadDeformedGeometry=true
      bRequireCurrentFrameData=false
      ComponentTags []
      FilteredBones []
      FilteredSockets [JNT_SpeederSocket, JNT_SpeederSocket_0, JNT_SpeederSocket_1]
      MeshUserParameter {parameter {name=None, typeDefHandle {registeredTypeIndex=1}, varData []}}
      SamplingRegions []
     }}
    Socket Sampling Mode=::NewEnumerator5
    Use Sampled Mesh Attribute as Ribbon ID=false
   )
   &ScaleMeshSize(Scale Factor=&VectorFromCurve(
     VectorCurve=DataInterface() {propertyValues {
       XCurve {defaultValue=0.4392, keys [{tangentMode=RCTM_User, time=0, value=0.5}, {tangentMode=RCTM_User, time=1, value=1}]}
       YCurve {defaultValue=1, keys [{tangentMode=RCTM_User, time=0, value=0.5}, {tangentMode=RCTM_User, time=1, value=1}]}
       ZCurve {defaultValue=1, keys [{tangentMode=RCTM_User, time=0, value=0.5}, {tangentMode=RCTM_User, time=1, value=1}]}
      }}
     Scale Curve=<0.8 0.8 0.8>
    ))
   &ScaleColor(Scale Mode=2)
   /SetVariables(Particles.Vector:Vector3f=&VectorFromCurve(
     VectorCurve=DataInterface() {propertyValues {
       XCurve {defaultValue=1, keys [{tangentMode=RCTM_User, time=0, value=0.5}, {tangentMode=RCTM_User, time=1, value=1}]}
       YCurve {defaultValue=1, keys [{tangentMode=RCTM_User, time=0, value=0.5}, {tangentMode=RCTM_User, time=1, value=1}]}
       ZCurve {defaultValue=1, keys [{tangentMode=RCTM_User, time=0, value=0.5}, {tangentMode=RCTM_User, time=1, value=1}]}
      }}
     Scale Curve=<1.2 1.2 1.2>
    )) {isSetParameters=true}
   /SetVariables(Particles.LightColor=&ColorFromCurve(ColorCurve=DataInterface() {propertyValues {
       AlphaCurve {keys [<at 0 0.189 RCIM_Cubic>,<at 1 0 RCIM_Cubic>]}
       BlueCurve {keys [<at 0 1 RCIM_Cubic>,<at 1 1 RCIM_Cubic>]}
       GreenCurve {keys [<at 0 1 RCIM_Cubic>,<at 1 1 RCIM_Cubic>]}
       RedCurve {keys [<at 0 1 RCIM_Cubic>,<at 1 1 RCIM_Cubic>]}
      }})) {isSetParameters=true}
   &SolveForcesAndVelocity(Clamp Velocity=false, Limit Acceleration=false)
   &FN_Light_Attributes(
    Color=<rgba 0.413 3 0 0.3652>
    Create Color Attribute=true
    Create Position Attribute=true
    Create Radius Attribute=true
    Position=&AddVectorToPosition(Position=$Particles.Position, Vector=<0 0 -11.1552>)
    Radius=58.2785
    Scale Color=0.488
   )
  ]
  rendererClasses [
   &NiagaraMeshRendererProperties
   &NiagaraSpriteRendererProperties
   &NiagaraLightRendererProperties
  ]
  renderers [
   {rendererClass=&NiagaraMeshRendererProperties, rendererIndex=0}
   {rendererClass=&NiagaraMeshRendererProperties, rendererIndex=1}
   {rendererClass=&NiagaraSpriteRendererProperties, rendererIndex=2}
   {rendererClass=&NiagaraLightRendererProperties, rendererIndex=3}
  ]
 }
 LargeWake002 {
  emitterUpdateScript [
   &EmitterState(Inactive Response=::NewEnumerator2, Life Cycle Mode=::NewEnumerator1, Loop Behavior=::NewEnumerator1, Loop Duration=1, UseLoopDelay=true)
   &SpawnBurst_Instantaneous(Spawn Count=8, Use Spawn Probability=false, UseLoopCountLimit=false)
   &SpawnBurst_Instantaneous(Spawn Count=2, Spawn Group=1, Use Spawn Probability=false, UseLoopCountLimit=false) {module=SpawnBurst_Instantaneous001}
   &SpawnRate(
    SpawnGroup=1
    SpawnRate=&Lerp_Float(A=41.2082, B=15, Alpha=$Emitter.InterpolateSpeedThresholdEmitter_01.InterpolateOverTime_Float.Moving Average)
    Use Spawn Probability=false
   ) {enabled=false, module=SpawnRate001}
  ]
  particleSpawnScript [
   /SetVariables(Particles.VisibilityTag=&SpawnGroup()) {isSetParameters=true, module=SetVariables_#1}
   &PartitionParticles(Number Of Partitions=&MaskIntBySpawnGroup(Masked Int=8, Passthrough Int=2, Spawn Group Mask=0))
   &InitializeParticle(
    Color=&MergeHelper_ColorSwitch(
     0=<rgba 0.413 3 0 0.2>
     1=<rgba 0.413 3 0 0.2>
     2=<rgba 0.413 3 0 0.2>
     3=<rgba 0.413 3 0 0.1>
     4=<rgba 0.413 3 0 0.1>
     Index=$Particles.PartitionParticles.Partition
     Num=::NewEnumerator3
    )
    Lifetime=0.3
    Mesh Scale Mode=::NewEnumerator4
    Position Offset=&MakeVector(X=0, Y=0, Z=-80)
    Sprite Size Mode=::NewEnumerator3
    Uniform Sprite Size=&MergeHelper_FloatSwitch(0=23.4132, 1=23.4132, 2=23.4132, 3=26.238, 4=26, Index=$Particles.PartitionParticles.Partition, Num=::NewEnumerator3)
    UsePositionOffset=true
    Write Material Random=true
   )
   &CameraOffset(Camera Offset Amount=&MergeHelper_FloatSwitch(0=5, 1=5, 2=5, 3=10, 4=10, Index=$Particles.PartitionParticles.Partition, Num=::NewEnumerator3))
  ]
  particleUpdateScript [
   &ParticleState(Kill Particles When Lifetime Has Elapsed=false)
   &SkeletalMeshLocation(
    Bone / Socket Index=&MergeHelper_IntSwitch(
     0=&MergeHelper_IntSwitch(0=0, 1=1, 2=2, 3=3, 4=4, 5=?, Index=$Particles.PartitionParticles.Partition, Num=::NewEnumerator3)
     1=&MergeHelper_IntSwitch(0=5, 1=6, 2=?, 3=?, 4=?, 5=?, Index=$Particles.PartitionParticles.Partition, Num=::NewEnumerator0)
     2=?
     3=?
     4=?
     5=?
     Index=&SpawnGroup()
     Num=::NewEnumerator0
    )
    MaskSpawnGroup=false
    Mesh Sampling Type=::NewEnumerator5
    Sampled Position Offset=&MergeHelper_VectorSwitch(
     0=&MergeHelper_VectorSwitch(
      0=<0 0 0>
      1=<0 0 0>
      2=<0 0 0>
      3=<0 0 0>
      4=<0 0 0>
      5=<-20 140 0>
      6=<20 140 0>
      7=<0 140 0>
      Index=$Particles.PartitionParticles.Partition
      Num=::NewEnumerator6
     )
     1=<0 0 0>
     2=?
     3=?
     Index=&SpawnGroup()
     Num=::NewEnumerator0
    )
    Skeletal Mesh=DataInterface() {propertyValues {
      DefaultMesh=null
      ExcludeBoneName=root
      OptionalComponentClass=null
      PreviewMesh=&FearCatch_Gadget
      SkinningMode=SkinOnTheFly
      SoftSourceActor=null
      SourceMode=Default
      UvSetIndex=0
      WholeMeshLOD=-1
      bExcludeBone=false
      bReadDeformedGeometry=true
      bRequireCurrentFrameData=false
      ComponentTags []
      FilteredBones []
      FilteredSockets [
       JNT_SpeederSocket
       JNT_SpeederSocket_0
       JNT_SpeederSocket_1
       JNT_SpeederSocket_2
       JNT_SpeederSocket_3
       JNT_SpeederSocket_4
       JNT_SpeederSocket_5
      ]
      MeshUserParameter {parameter {name=None, typeDefHandle {registeredTypeIndex=1}, varData []}}
      SamplingRegions []
     }}
    Socket Sampling Mode=::NewEnumerator5
    Use Sampled Mesh Attribute as Ribbon ID=false
   )
   &SolveForcesAndVelocity(Clamp Velocity=false, Limit Acceleration=false)
   &FN_Light_Attributes(
    Color=<rgba 0.413 3 0 0.3652>
    Create Color Attribute=true
    Create Position Attribute=true
    Create Radius Attribute=true
    Position=&AddVectorToPosition(Position=$Particles.Position, Vector=<0 0 -11.1552>)
    Radius=58.2785
    Scale Color=0.488
   )
   &ScaleColor(
    Scale Alpha=&Waveform("Waveform [1]"=::NewEnumerator0, "[1] Frequency"=0.656, "[1] Amplitude Min/Max"=<0 1>, "[1] Scale"=0.384, "[1] Bias"=0.84)
    ScaleA=true
    ScaleRGB=true
   )
   /SetVariables(
    Particles.NoCamColor:LinearColor=&Multiply_LinearColorByFloat(
     Linear Color=<rgba 0.413 3 0 0.1>
     Float=&Waveform("Waveform [1]"=::NewEnumerator0, "[1] Frequency"=0.656, "[1] Amplitude Min/Max"=<0 1>, "[1] Scale"=0.384, "[1] Bias"=0.84)
    )
    Particles.NoCamSize:Vector2f=<39.9267 140>
    Particles.SpriteAlignNoCam:Vector3f=&RotateVectorByQuaternion(Vector To Rotate=<0 1 0>, Quaternion=$Particles.SkeletalMeshLocation.SampledRotation)
   ) {isSetParameters=true}
  ]
  rendererClasses [&NiagaraSpriteRendererProperties, &NiagaraLightRendererProperties]
  renderers [
   {rendererClass=&NiagaraSpriteRendererProperties}
   {rendererClass=&NiagaraLightRendererProperties, rendererIndex=1}
   {rendererClass=&NiagaraSpriteRendererProperties, rendererIndex=2}
  ]
 }
 Smoke {
  emitterUpdateScript [
   &EmitterState(Life Cycle Mode=::NewEnumerator1, Loop Duration=1.1, UseLoopDelay=false)
   &self:InterpolateSpeedThresholdEmitter_03(
    Define Initial Value=false
    Rate Of Change=5
    Target Float Value:NiagaraFloat=&MakeCustomFloatFromBool(
     Bool=&SetBoolByFloatComparison(
      A=&Add_Float(
       A=&Add_Float(
        A=&Abs_Float(Float:NiagaraFloat=&MakeFloatFromVector(Vector=$Engine.Owner.Velocity, Channel=::NewEnumerator0))
        B=&Abs_Float(Float:NiagaraFloat=&MakeFloatFromVector(Vector=$Engine.Owner.Velocity, Channel=::NewEnumerator2))
       )
       B=&Abs_Float(Float:NiagaraFloat=&MakeFloatFromVector(Vector=$Engine.Owner.Velocity, Channel=::NewEnumerator1))
      )
      B=46.278
     )
     True Float=1
     False Float=0
    )
   ) {module=InterpolateSpeedThresholdEmitter_03}
   &SpawnBurst_Instantaneous(Spawn Group=1, Use Spawn Probability=false, UseLoopCountLimit=false)
   &SpawnRate(
    SpawnRate=&Lerp_Float(A=10, B=0, Alpha=$Emitter.InterpolateSpeedThresholdEmitter_01.InterpolateOverTime_Float.Moving Average)
    Use Spawn Probability=false
   )
  ]
  particleSpawnScript [
   /SetVariables(Particles.VisibilityTag=&SpawnGroup()) {isSetParameters=true}
   /SetVariables(Particles.Random=&RandomRangeFloat(Minimum=0, Maximum=1)) {isSetParameters=true}
   &CameraOffset(Camera Offset Amount=30)
   &DynamicMaterialParameters(
    Index 0 Param 2=&UniformAOrBFloat(A=0)
    Module.Param0WriteEnabled=true
    Module.Param1WriteEnabled=true
    Module.Param2WriteEnabled=true
    Module.Param3WriteEnabled=true
   ) {module=DynamicMaterialParameters001}
   &InitializeParticle(
    Color=<rgba 0.506 1 0.533 1>
    Lifetime=1.1
    Position Offset=<0 0 -100>
    Sprite Rotation Mode=::NewEnumerator2
    Sprite Size Mode=::NewEnumerator3
    Uniform Sprite Size=&Lerp_Float(A=120, B=170, Alpha=$Particles.Random)
    UsePositionOffset=true
    Write Material Random=true
   ) {module=InitializeParticle001}
   &ApplyInitialForces()
   &InitialMeshOrientation(Mesh Orientation Mode=::NewEnumerator2, Orientation Coordinate Space=::World, Rotation=<0 -0.25 0>, Use Rotation Vector=true)
  ]
  particleUpdateScript [
   &SkeletalMeshLocation(
    Bone Sampling Mode=::NewEnumerator4
    MaskSpawnGroup=false
    Orientation Sampling=::NewEnumerator1
    Position Sampling=::NewEnumerator2
    Sampled Position Offset=<0 -20 -30>
    Skeletal Mesh=DataInterface() {propertyValues {
      DefaultMesh=null
      ExcludeBoneName=root
      OptionalComponentClass=null
      PreviewMesh=&FearCatch_Gadget
      SkinningMode=SkinOnTheFly
      SoftSourceActor=null
      SourceMode=Default
      UvSetIndex=0
      WholeMeshLOD=-1
      bExcludeBone=false
      bReadDeformedGeometry=true
      bRequireCurrentFrameData=false
      ComponentTags []
      FilteredBones [JNT_Speeder]
      FilteredSockets []
      MeshUserParameter {parameter {name=None, typeDefHandle {registeredTypeIndex=1}, varData []}}
      SamplingRegions []
     }}
    Use Sampled Mesh Attribute as Ribbon ID=false
   )
   &ParticleState()
   &ScaleSpriteSize(Uniform Curve Sprite Scale=DataInterface() {propertyValues {Curve {keys [
        {arriveTangent=1.6815, interpMode=RCIM_Cubic, leaveTangent=1.6815, tangentMode=RCTM_User, time=0.2, value=0.5}
        {arriveTangent=0.1063, interpMode=RCIM_Cubic, leaveTangent=0.1063, tangentMode=RCTM_User, time=0.55, value=1}
       ]}}})
   &SolveForcesAndVelocity(Clamp Velocity=false, Limit Acceleration=false)
   &AlignSpriteToMeshOrientation(SetAlignment=true, SetFacing=true)
   &DynamicMaterialParameters(
    Index 0 Param 1=&FloatFromCurve(FloatCurve=DataInterface() {propertyValues {Curve {keys [
         {arriveTangent=1.7661, interpMode=RCIM_Cubic, leaveTangent=1.7661, tangentMode=RCTM_User, time=0, value=0}
         {arriveTangent=0.0224, interpMode=RCIM_Cubic, leaveTangent=0.0224, tangentMode=RCTM_User, time=1, value=1}
        ]}}})
    Index 0 Param 2=1
    Module.Param0WriteEnabled=true
    Module.Param1WriteEnabled=true
    Module.Param2WriteEnabled=true
    Module.Param3WriteEnabled=true
   )
   &ScaleColor(
    Scale Alpha=&FloatFromCurve(FloatCurve=DataInterface() {propertyValues {Curve {keys [<at 0 0>, {tangentMode=RCTM_User, time=0.75, value=1}, {tangentMode=RCTM_User, time=1, value=0}]}}})
    Scale RGB=&MakeVectorFromLinearColorRGB(Color=&ColorFromCurve(ColorCurve=DataInterface() {propertyValues {
        AlphaCurve {keys [<at 0 0>,<at 0.4487 1 RCIM_Cubic>]}
        BlueCurve {keys [
          <at 0.1811 0.6425 RCIM_Cubic>
          {arriveTangent=-0.3057, interpMode=RCIM_Cubic, leaveTangent=-0.3057, time=0.4806, value=1}
          <at 0.7837 0.4583>
         ]}
        GreenCurve {keys [
          <at 0.1811 0.6562 RCIM_Cubic>
          {arriveTangent=-0.3285, interpMode=RCIM_Cubic, leaveTangent=-0.3285, time=0.4806, value=1}
          <at 0.7837 0.4583>
         ]}
        RedCurve {keys [
          <at 0.1811 0.6388 RCIM_Cubic>
          {arriveTangent=-0.2996, interpMode=RCIM_Cubic, leaveTangent=-0.2996, time=0.4806, value=1}
          <at 0.7837 0.4583>
         ]}
       }}))
    ScaleA=true
    ScaleRGB=true
   )
   &FN_Light_Attributes(
    Color=<rgba 0.0113 10 0 1>
    Create Color Attribute=true
    Create Position Attribute=true
    Create Radius Attribute=true
    Position:NiagaraPosition=$Particles.SkeletalMeshLocation.SampledPosition
    Radius=60
   )
  ]
  rendererClasses [&NiagaraSpriteRendererProperties, &NiagaraLightRendererProperties]
  renderers [
   {rendererClass=&NiagaraSpriteRendererProperties}
   {rendererClass=&NiagaraLightRendererProperties, rendererIndex=1}
  ]
 }
}
summary {
 systemName=NS_Emote_FearCatch_Thrusters
 emitters [
  {emitterName=Smoke, rendererClasses [&NiagaraSpriteRendererProperties, &NiagaraLightRendererProperties]}
  {emitterName=LargeWake, rendererClasses [&NiagaraMeshRendererProperties]}
  {
   emitterName=LargeWake001
   rendererClasses [
    &NiagaraMeshRendererProperties
    &NiagaraSpriteRendererProperties
    &NiagaraLightRendererProperties
   ]
  }
  {emitterName=LargeWake002, rendererClasses [&NiagaraSpriteRendererProperties, &NiagaraLightRendererProperties]}
 ]
 userVariables [
  User.bIsFrontEnd {type=NiagaraBool}
  User.bIsFrontEndPreview {type=NiagaraBool}
 ]
}
